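# Translucent, two-sided unlit effect material for the Canary quest Object of Power combine effect, usable on Niagara meshes, ribbons, sprites and beam trails.
# Sphere-mask shapes plus two panning noise textures form an alpha, shaped along U by sine and dynamic-parameter falloff, with particle alpha and optional depth fade.
# Particle color feeds an emissive adjustment function, while a separate panning blurred-noise branch computes a world position offset amount.
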Material /Game/Environments/Artemis/Props/Quest/Quest_Canary_Object_Of_Power/FX/M_S21_CombineEffect_Mesh
graph {
 properties {
  bUsedWithBeamTrails=true
  bUsedWithNiagaraMeshParticles=true
  bUsedWithNiagaraRibbons=true
  bUsedWithParticleSprites=true
  blendMode=BLEND_Translucent
  pixelDepthOffsetMode=PDOM_Legacy
  shadingModel=MSM_Unlit
  twoSided=true
  usedShadingModels=Unlit
 }
 nodes [
  n1 = ComponentMask(None=TextureCoordinate()) {r=true}
  n2 = Sine(None=Multiply(A=@n1) {constB=0.5})
  n3 = OneMinus(None=@n1)
  n4 = DynamicParameter() {defaultValue=<rgba 2 1 1 1>, paramNames [Mask,Param2,Param3,Param4]}
  n5 = ParticleColor()
  n6 = TextureCoordinate()
  n7 = Multiply(A=OneMinus(None=MaterialFunctionCall(Float2=@n6) {materialFunction=&BreakOutFloat2Components}.R))
  n8 = SphereMask(A=TextureCoordinate(), B=Constant2Vector() {g=0.7, r=0.5}) {attenuationRadius=0.85, hardnessPercent=0}
  n9 = TextureCoordinate()
  n10 = Constant2Vector() {g=0.5, r=1.9}
  n11 = Multiply(
   A=Multiply(A=Subtract(A=@n7, B=@n8), B=SphereMask(A=@n9, B=@n10) {attenuationRadius=2, hardnessPercent=95})
   B=SphereMask(A=TextureCoordinate() {uTiling=0.3}, B=Constant2Vector() {g=0.5}) {attenuationRadius=0.5, hardnessPercent=70}
  )
  n12 = TextureCoordinate() {uTiling=0.5, vTiling=8}
  n13 = Constant2Vector() {g=1.5, r=1}
  n14 = ScalarParameter() {defaultValue=1, parameterName=NoiseUtliing}
  n15 = ScalarParameter() {defaultValue=1, parameterName=NoiseVtliing}
  n16 = Panner(
   Coordinate=Multiply(
    A=MaterialFunctionCall(UVs=@n12, Texture Scale=@n13) {materialFunction=&ScaleUVsByCenter}.UVs
    B=AppendVector(A=@n14, B=@n15)
   )
   Speed=AppendVector(A=ScalarParameter() {defaultValue=1.5, parameterName=PanSpeedU}, B=ScalarParameter() {parameterName=PanSpeedV})
  ) {speedX=0.5}
  n17 = MaterialFunctionCall(UVs=TextureCoordinate() {coordinateIndex=1, uTiling=0.5, vTiling=1.5}, Texture Scale=Constant2Vector() {g=1.5, r=1}) {materialFunction=&ScaleUVsByCenter}
  n18 = AppendVector(A=ScalarParameter() {defaultValue=1, parameterName=NoiseUtliing}, B=ScalarParameter() {defaultValue=1, parameterName=NoiseVtliing})
  n19 = ScalarParameter() {defaultValue=1.5, parameterName=PanSpeedU}
  n20 = ScalarParameter() {parameterName=PanSpeedV}
  n21 = TextureSample(UVs=Panner(Coordinate=Multiply(A=@n17.UVs, B=@n18), Speed=AppendVector(A=@n19, B=@n20)) {speedX=0.5}) {texture=&T_Inky_Smoke_Tile_Chromatic}
  n22 = Add(
   A=Multiply(A=TextureSample(UVs=@n16) {samplerSource=SSM_Wrap_WorldGroupSettings, texture=/Game/Weapons/FORT_Melee/Effects/LaserSwordFounders/T_LaserSwordTrailNoise}.RGB) {constB=7}
   B=Power(Base=Multiply(A=@n21.RGB) {constB=3}) {constExponent=5}
  )
  n23 = Saturate(None=Add(A=Saturate(None=@n11), B=Multiply(A=@n22) {constB=0.8}))
  n24 = Multiply(A=Saturate(None=Multiply(A=@n2) {constB=1.5}), B=Multiply(A=Power(Base=@n3, Exp=@n4.Mask), B=Multiply(A=@n5.A, B=@n23)))
  n25 = ScalarParameter() {parameterName=Depth Fade Diatnce}
  LinearInterpolate(A=@n24, B=Multiply(A=DepthFade(FadeDistance=@n25), B=@n24), Alpha=ScalarParameter() {parameterName=Use Depth Fade})
  MaterialFunctionCall(Emissive=ParticleColor().RGB) {materialFunction=&MF_FX_EmissiveAdjustments}
  n26 = TextureCoordinate()
  n27 = ComponentMask(None=@n26) {r=true}
  n28 = ScalarParameter() {defaultValue=0.25, group=WPO, parameterName=WPO Tile U}
  n29 = ComponentMask(None=@n26) {g=true}
  n30 = ScalarParameter() {defaultValue=0.25, group=WPO, parameterName=WPO Tile V}
  n31 = Panner(
   Coordinate=AppendVector(A=Multiply(A=@n27, B=@n28), B=Multiply(A=@n29, B=@n30))
   Time=Time()
   Speed=AppendVector(
    A=ScalarParameter() {group=WPO, parameterName=WPO Pan Speed U}
    B=ScalarParameter() {defaultValue=0.2, group=WPO, parameterName=WPO Pan Speed V}
   )
  ) {speedX=-1, speedY=0.1}
  Multiply(
   A=ScalarParameter() {defaultValue=0.8, group=WPO, parameterName=WPO Amnt Noise}
   B=MaterialFunctionCall(
    In=TextureSample(UVs=@n31) {samplerType=SAMPLERTYPE_Grayscale, texture=/FortWater/Textures/LowResBlurredNoise}.R
    Contrast=ScalarParameter() {parameterName=NoiseContrast}
   ) {materialFunction=&CheapContrast}.Result
  )
  ScalarParameter() {defaultValue=2, parameterName=Mask_Power}
 ]
}
